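# Unlit translucent material for misty ribbon or beam trails in the Blaster enemy's effects.
# Two noise channels pan on dynamic-parameter time with separate tiling and speeds.
# A squared-sine edge mask minus AlphaOffset-biased noise, with contrast shifting over particle lifetime, forms the eroded opacity.
# That mask and the second noise blend particle-tinted bright and dark MF_BlasterCommon colors, scaled by ScaleColor, into emissive.
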Material /Game/Characters/Enemies/Blaster/FX/M_RibbonMist
graph {
 properties {bUsedWithBeamTrails=true, blendMode=BLEND_Translucent, pixelDepthOffsetMode=PDOM_Legacy, shadingModel=MSM_Unlit, usedShadingModels=Unlit}
 nodes [
  n1 = TextureCoordinate() {coordinateIndex=1}
  n2 = MaterialFunctionCall(Float2=@n1) {materialFunction=&BreakOutFloat2Components}
  n3 = VectorParameter() {parameterName=UV_Speed, defaultValue=<rgba 0.2 0.35 0.4 0.2>}
  AppendVector(A=Multiply(A=@n2.R, B=@n3.G), B=@n2.G)
  AppendVector(A=Multiply(A=@n2.R, B=@n3.R), B=@n2.G)
  n4 = MaterialFunctionCall!(TimeScale, ElementColor, HeadEnhanceScale) {materialFunction=/Game/Effects/Fort_Effects/Materials/_Functions/MF_BlasterCommon}
  n5 = ParticleColor()
  n6 = Multiply(A=@n4.BrightColor, B=StaticSwitchParameter(True=@n5.RGB, False=@n5.RGB) {parameterName=UseParticleAlphaForAlphaOnly})
  n7 = Multiply(A=@n1, B=@n3.R)
  n8 = DynamicParameter() {paramNames [Time,ScaleColor,Emissive,AlphaOffset]}
  n9 = AppendVector(A=@n3.B, B=Constant())
  n10 = MaterialFunctionCall(Float2=Sine(None=TextureCoordinate() {uTiling=0.5, vTiling=0.5})) {materialFunction=&BreakOutFloat2Components}
  n11 = Multiply(A=@n10.R, B=@n10.R)
  n12 = Panner(Coordinate=Multiply(A=@n1, B=@n3.G), Time=@n8.Time, Speed=AppendVector(A=@n3.A, B=Constant()))
  n13 = Subtract(
   A=Saturate(None=Multiply(A=@n11) {constB=1.5})
   B=LinearInterpolate(
    A=@n8.AlphaOffset
    Alpha=TextureSample(UVs=@n12) {samplerType=SAMPLERTYPE_LinearColor, texture=&T_NoisePack_LinearColor}.R
   )
  )
  n14 = MaterialFunctionCall(In=@n13, Contrast=LinearInterpolate(Alpha=ParticleRelativeTime()) {constA=0.35, constB=-0.35}) {materialFunction=&CheapContrast}
  n15 = LinearInterpolate(
   A=TextureSample(UVs=Panner(Coordinate=@n7, Time=@n8.Time, Speed=@n9)) {samplerType=SAMPLERTYPE_LinearColor, texture=&T_NoisePack_LinearColor}.G
   B=@n14.Result
  ) {constAlpha=0.6}
  MaterialFunctionCall(Emissive=Multiply(A=LinearInterpolate(A=@n6, B=@n4.DarkColor, Alpha=@n15), B=DynamicParameter() {paramNames [Time,ScaleColor,Emissive,AlphaOffset]}.ScaleColor)) {materialFunction=&MF_FX_EmissiveAdjustments}
  Multiply(A=@n14.Result, B=StaticSwitchParameter(True=ParticleColor().A, False=Constant() {r=1}) {parameterName=UseParticleAlphaForAlphaOnly})
  Saturate(None=@n13)
  ScalarParameter() {defaultValue=0.2, parameterName=Contrast}
 ]
}
